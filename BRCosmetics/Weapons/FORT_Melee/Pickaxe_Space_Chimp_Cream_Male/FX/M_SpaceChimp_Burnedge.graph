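# Additive, unlit, two-sided material for Niagara mesh particles, drawing a smoky burning-edge effect for the Space Chimp pickaxe.
# Two samples of an inky smoke texture pan in opposite U directions and blend across U, splitting the flow about the middle.
# Zero-tiled UVs turn sphere masks into soft U and V falloffs shaping the smoke.
# Powered results, tinted by particle color and alpha, drive emissive (through MF_FX_EmissiveAdjustments) and opacity.
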Material /BRCosmetics/Weapons/FORT_Melee/Pickaxe_Space_Chimp_Cream_Male/FX/M_SpaceChimp_Burnedge
graph {
 properties {
  bUsedWithNiagaraMeshParticles=true
  blendMode=BLEND_Additive
  pixelDepthOffsetMode=PDOM_Legacy
  previewMesh=&/SM_SpaceChimp_PickAxe_BurnEdge
  shadingModel=MSM_Unlit
  twoSided=true
  usedShadingModels=Unlit
 }
 nodes [
  n1 = TextureCoordinate() {vTiling=1.5}
  n2 = ComponentMask(None=Add(A=TextureCoordinate()) {constB=-0.25}) {r=true}
  n3 = LinearInterpolate(
   A=TextureSample(UVs=Panner(Coordinate=@n1) {speedX=2}) {texture=&T_Inky_Smoke_Tile}.R
   B=TextureSample(UVs=Panner(Coordinate=@n1) {speedX=-2}) {texture=&T_Inky_Smoke_Tile}.R
   Alpha=Saturate(None=Multiply(A=@n2) {constB=2})
  )
  n4 = SphereMask(A=TextureCoordinate() {vTiling=0}, B=Constant2Vector() {r=0.5}, Radius=Constant() {r=0.5}, Hardness=Constant())
  n5 = Multiply(A=@n4) {constB=2}
  n6 = Multiply(
   A=Multiply(A=Add(A=@n3, B=@n4), B=Saturate(None=@n5))
   B=SphereMask(A=TextureCoordinate() {uTiling=0}, B=Constant2Vector() {g=0.5}, Radius=Constant() {r=0.45}, Hardness=Constant() {r=0.6})
  )
  n7 = Power(Base=@n6) {constExponent=4}
  n8 = ParticleColor()
  MaterialFunctionCall(Emissive=Multiply(A=Saturate(None=@n7), B=@n8.RGB)) {materialFunction=&MF_FX_EmissiveAdjustments}
  Multiply(A=Saturate(None=Power(Base=@n6)), B=@n8.A)
 ]
}
