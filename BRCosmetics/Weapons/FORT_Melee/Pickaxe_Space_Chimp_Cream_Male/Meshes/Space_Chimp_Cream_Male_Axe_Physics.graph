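# Physics asset for the Space Chimp Cream Male pickaxe mesh, a cosmetic two-handed melee weapon.
# It defines a capsule body on the handle bone plus one on each of two dynamic tube bones, back and left.
# Each tube is constrained to the handle with angular swing and twist limits, wide on one swing axis and tighter on the others, with the back tube allowed the looser range.
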PhysicsAsset /BRCosmetics/Weapons/FORT_Melee/Pickaxe_Space_Chimp_Cream_Male/Meshes/Space_Chimp_Cream_Male_Axe_Physics
bodies {
 Melee_TwoHanded_Handle [{shapeType=Capsule, length=5, radius=5, shapeName=Melee_TwoHanded_Handle_capsule}]
 dyn_tube_bk [{shapeType=Capsule, length=5, radius=5, shapeName=dyn_tube_bk_capsule, center=<10 0 0>}]
 dyn_tube_l [{shapeType=Capsule, length=5, radius=5, shapeName=dyn_tube_l_capsule, center=<10 0 0>}]
}
constraints [
 {bone1Name=dyn_tube_bk, bone2Name=Melee_TwoHanded_Handle, swing1LimitDegrees=165, swing2LimitDegrees=45, twistLimitDegrees=45}
 {bone1Name=dyn_tube_l, bone2Name=Melee_TwoHanded_Handle, swing1LimitDegrees=150, swing2LimitDegrees=30, twistLimitDegrees=30}
]
